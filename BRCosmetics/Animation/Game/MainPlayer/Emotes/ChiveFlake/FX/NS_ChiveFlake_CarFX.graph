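# Niagara effect for the ChiveFlake emote's car gadget that kicks up dust and rocks from its wheels.
# Front and back wheel pairs each layer toon-shaded mesh puffs with sprites, using two spawn groups to sample wheel bones.
# Gameplay dust rate follows a system speed threshold, while rocks spawn per distance moved from rear wheels, with platform-gated collision.
# Front-end modifiers swap spawn rates and velocity cones, and time of day tints particles.
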NiagaraSystem /BRCosmetics/Animation/Game/MainPlayer/Emotes/ChiveFlake/FX/NS_ChiveFlake_CarFX
dependencies {
 usedDataInterfaces [
  &NiagaraDataInterfaceSkeletalMesh
  &NiagaraDataInterfaceCurve
  &NiagaraDataInterfacePlatformSet
  &NiagaraDataInterfaceColorCurve
  &NiagaraDataInterfaceVectorCurve
  &NiagaraDataInterfaceArrayFloat
 ]
 usedDynamicInputs [
  &Multiply_Vector2D
  &MakeVector2D
  &MakeCustomFloatFromBool
  &Add_Float
  &Abs_Float
  &MakeFloatFromVector
  &ClampFloat
 ]
 usedModules [
  &self:SystemSpawnScript.SystemScriptSource.SystemScriptGraph.NiagaraNodeAssignment_7.SetVariables_#1
  &SystemState
  &self:SystemSpawnScript.SystemScriptSource.SystemScriptGraph.NiagaraNodeAssignment_8.SetVariables_#2
 ]
}
emitters {
 BackWheel {
  emitterSpawnScript [/Game/Effects/Niagara/MasterMaterials/ToonShader/Template/NM_ToonShaderParameters(BaseColor:LinearColor=<rgba 0.6157 0.5412 0.3922 1>, Lighting_Hardness=0.2)]
  emitterUpdateScript [
   &EmitterState(Life Cycle Mode=::NewEnumerator1, Loop Behavior=::NewEnumerator1, Loop Delay=2.2, Loop Duration Mode=::NewEnumerator1, UseLoopDelay=false)
   &SpawnRate(
    SpawnRate=&FN_FrontEndFloatModifier(Gameplay=&Lerp_Float(A=0, B=64, Alpha=$System.Speed Threshold), Front End=64)
    Use Spawn Probability=false
   )
   &SpawnPerUnit(Module.Use Max Movement Threshold=true, Module.Use Movement Tolerance=false, Spawn Spacing=5, Use Spawn Probability=false) {enabled=false}
  ]
  particleSpawnScript [
   &InitializeParticle(
    Color=&Multiply_LinearColorByFloat(
     Linear Color=&FNTimeOfDayColor(Night Color=<rgba 0.7961 0.8667 1 1>, Morning Color=<rgba 1 1 1 1>, Day Color=<rgba 1 1 1 1>, Evening Color=<rgba 1 1 1 1>)
     Float=&MakeFloatFromLinearColor(
      LinearColor=&MakeCustomLinearColorFromBool(
       Bool=&SetBoolByFloatComparison(A=$User.FrontEnd, B=0)
       True Color=<rgba 1 1 1 1>
       False Color=<rgba 0.6 0.6 0.6 1>
      )
      Channel:ENiagaraLinearColor_Channels=::NewEnumerator0
     )
    )
    Lifetime Max=0.6
    Lifetime Min=0.4
    Lifetime Mode=::NewEnumerator1
    Mesh Scale=&Multiply_VectorByFloat(Vector=<1 1 1>, Float=&RandomRangeFloat(Minimum=1.2, Maximum=1.6))
    Mesh Scale Mode=::NewEnumerator4
    UsePositionOffset=false
    Write Material Random=true
   )
   &FN_SetSpawnGroup(Number of Spawn Groups=2)
   &SkeletalMeshLocation(
    Bone / Socket Index=&SpawnGroup()
    Bone Sampling Mode=::NewEnumerator4
    MaskSpawnGroup=false
    Sampled Position Offset=<0 -5 -5>
    Sampled Position Offset Coordinate Space:ENiagaraSkelMeshTransforms=::NewEnumerator2
    Skeletal Mesh=$System.Skeletal Mesh
   )
   &InitialMeshOrientation(Rotation=&RandomRangeVector(Minimum=<0 0 0>, Maximum=<1 1 1>), Use Rotation Vector=true)
   &ShapeLocation(MaskSpawnGroup=false, Override Seed=false, Sphere Radius=2, UseOwnerScale=true)
   &AddVelocity(
    Cone Angle=&FN_FrontEndFloatModifier(Gameplay=75, Front End=30)
    Cone Axis=<0 -1 0>
    Module.UseConeFalloff=false
    Module.UseVelDistribution=false
    Override Seed=false
    Velocity Mode=::NewEnumerator2
    Velocity Speed=&FN_FrontEndFloatModifier(
     Gameplay=&RandomRangeFloat(Minimum=245, Maximum=280)
     Front End=&RandomRangeFloat(Minimum=500, Maximum=650)
    )
   )
  ]
  particleUpdateScript [
   &ParticleState()
   &ScaleMeshSize(Scale Factor=&VectorFromFloat(Value=&FloatFromCurve(FloatCurve=DataInterface() {propertyValues {Curve {keys [
          {arriveTangent=0.0398, interpMode=RCIM_Cubic, leaveTangent=0.0398, tangentMode=RCTM_User, time=0, value=0.25}
          {arriveTangent=0.6083, interpMode=RCIM_Cubic, leaveTangent=0.6083, tangentMode=RCTM_User, time=0.2, value=0.35}
          {arriveTangent=2.388, interpMode=RCIM_Cubic, leaveTangent=2.388, tangentMode=RCTM_User, time=1, value=1.2}
         ]}}})))
   &ScaleColor(
    Scale Alpha=&FloatFromCurve(FloatCurve=DataInterface() {propertyValues {Curve {keys [
         {arriveTangent=0.7143, time=0.2, value=1}
         {arriveTangent=0.0076, interpMode=RCIM_Cubic, leaveTangent=0.0076, tangentMode=RCTM_User, time=0.7, value=1}
         {arriveTangent=-7.4459, interpMode=RCIM_Cubic, leaveTangent=-7.4459, tangentMode=RCTM_User, time=1, value=0}
        ]}}})
    Scale RGB=&Multiply_Vector(
     A=&MakeVectorFromLinearColorRGB(Color=&ColorFromCurve(ColorCurve=DataInterface() {propertyValues {
         AlphaCurve {keys [<at 0 1 RCIM_Cubic>]}
         BlueCurve {keys [<at 0 1 RCIM_Cubic>,<at 1 0.5882>]}
         GreenCurve {keys [<at 0 1 RCIM_Cubic>,<at 1 0.5882>]}
         RedCurve {keys [<at 0 1 RCIM_Cubic>,<at 1 0.5882>]}
        }}))
     B=&MakeVectorFromLinearColorRGB(Color=$System.TOD Color)
    )
    ScaleA=false
    ScaleRGB=true
   )
   &GravityForce(Gravity=&VectorFromCurve(
     VectorCurve=DataInterface() {propertyValues {
       XCurve {keys [<at 0 0>,<at 1 0>]}
       YCurve {keys [<at 0 0>,<at 1 0>]}
       ZCurve {keys [
         {arriveTangent=-0.1394, interpMode=RCIM_Cubic, leaveTangent=-0.1394, tangentMode=RCTM_User, time=0, value=0}
         {arriveTangent=9.9758, interpMode=RCIM_Cubic, leaveTangent=9.9758, tangentMode=RCTM_User, time=0.3, value=1}
        ]}
      }}
     Scale Curve=&RandomRangeVector(Minimum=<0 0 450>, Maximum=<0 0 575>)
    ))
   &Drag(
    Drag=&FloatFromCurve(
     FloatCurve=DataInterface() {propertyValues {Curve {keys [<at 0 0>,<at 0.05 1>]}}}
     Scale Curve=&RandomRangeFloat(Minimum=12, Maximum=16)
    )
    Use Linear Drag=true
    Use Rotational Drag=true
   )
   &InheritSourceMovement(
    Applied Position Scale=&VectorFromFloat(Value=&FloatFromCurve(
      FloatCurve=DataInterface() {propertyValues {Curve {keys [<at 0 0 RCIM_Cubic>, {arriveTangent=4.4031, interpMode=RCIM_Cubic, leaveTangent=4.4031, tangentMode=RCTM_User, time=0.1, value=1}]}}}
      Scale Curve=0.5
     ))
    ApplyForce=false
    ApplyPosition=true
    ApplyPositionFalloff=false
   )
   &SolveForcesAndVelocity(Clamp Velocity=false, Limit Acceleration=false)
   &DynamicMaterialParameters(
    Index 0 Param 1=&FloatFromCurve(FloatCurve=DataInterface() {propertyValues {Curve {keys [
         {arriveTangent=-0.0276, time=0, value=0}
         {arriveTangent=0.0397, interpMode=RCIM_Cubic, leaveTangent=0.0397, tangentMode=RCTM_User, time=0.65, value=0}
         {arriveTangent=3.5539, interpMode=RCIM_Cubic, leaveTangent=3.5539, tangentMode=RCTM_User, time=1, value=0.4}
        ]}}})
    Module.Index0Float0Write=true
    Module.Index0Float1Write=true
    Module.Index0Float2Write=true
    Module.Index0Float3Write=true
   ) {enabled=false}
   &DynamicMaterialParameters(
    Index 0 Param 1=&FloatFromCurve(FloatCurve=DataInterface() {propertyValues {Curve {keys [
         {arriveTangent=-0.0276, time=0, value=0}
         {arriveTangent=0.0397, interpMode=RCIM_Cubic, leaveTangent=0.0397, tangentMode=RCTM_User, time=0.6, value=0}
         {arriveTangent=2.6349, interpMode=RCIM_Cubic, leaveTangent=2.6349, tangentMode=RCTM_User, time=1, value=0.6}
        ]}}})
    Module.Index0Float0Write=true
    Module.Index0Float1Write=true
    Module.Index0Float2Write=true
    Module.Index0Float3Write=true
   ) {module=DynamicMaterialParameters001}
   &UpdateMeshOrientation(
    Rotation Rate=&FloatFromCurve(
     FloatCurve=DataInterface() {propertyValues {Curve {keys [{interpMode=RCIM_Cubic, tangentMode=RCTM_User, time=0.4, value=0}, {interpMode=RCIM_Cubic, tangentMode=RCTM_User, time=1, value=1}]}}}
     Scale Curve=&RandomRangeFloat(Minimum=-2, Maximum=2)
    )
    Rotation Vector=&RandomVector(Vector Scale=1)
   )
  ]
  rendererClasses [&NiagaraMeshRendererProperties]
  renderers [
   {rendererClass=&NiagaraMeshRendererProperties, rendererIndex=0}
   {rendererClass=&NiagaraMeshRendererProperties, rendererIndex=1}
  ]
 }
 BackWheelSPRITE {
  emitterUpdateScript [
   &EmitterState(Life Cycle Mode=::NewEnumerator1, Loop Behavior=::NewEnumerator1, Loop Delay=2.2, Loop Duration Mode=::NewEnumerator1, UseLoopDelay=false)
   &SpawnRate(
    SpawnRate=&FN_FrontEndFloatModifier(Gameplay=&Lerp_Float(A=0, B=32, Alpha=$System.Speed Threshold), Front End=64)
    Use Spawn Probability=false
   )
   &SpawnPerUnit(Module.Use Max Movement Threshold=true, Module.Use Movement Tolerance=false, Spawn Spacing=5, Use Spawn Probability=false) {enabled=false}
  ]
  particleSpawnScript [
   &InitializeParticle(
    Color=&MultiplyLinearColor(
     Linear Color=<rgba 0.9569 0.8275 0.5647 1>
     Scale Factor=&MakeVector4_FromLinearColor(LinearColor=&FNTimeOfDayColor(Night Color=<rgba 0.5765 0.6157 1 1>, Morning Color=<rgba 1 1 1 1>, Day Color=<rgba 1 1 1 1>, Evening Color=<rgba 1 1 1 1>))
    )
    Lifetime Max=0.6
    Lifetime Min=0.4
    Lifetime Mode=::NewEnumerator1
    Sprite Rotation Mode=::NewEnumerator2
    Sprite Size Mode=::NewEnumerator1
    Uniform Sprite Size Max=65
    Uniform Sprite Size Min=50
    UsePositionOffset=false
    Write Material Random=true
   )
   &FN_SetSpawnGroup(Number of Spawn Groups=2)
   &SkeletalMeshLocation(
    Bone / Socket Index=&SpawnGroup()
    Bone Sampling Mode=::NewEnumerator4
    MaskSpawnGroup=false
    Sampled Position Offset=<0 -5 -5>
    Sampled Position Offset Coordinate Space:ENiagaraSkelMeshTransforms=::NewEnumerator2
    Skeletal Mesh=$System.Skeletal Mesh
   )
   &ShapeLocation(MaskSpawnGroup=false, Override Seed=false, Sphere Radius=2, UseOwnerScale=true)
   &AddVelocity(
    Cone Angle=&FN_FrontEndFloatModifier(Gameplay=75, Front End=30)
    Cone Axis=<0 -1 0>
    Module.UseConeFalloff=false
    Module.UseVelDistribution=false
    Override Seed=false
    Velocity Mode=::NewEnumerator2
    Velocity Speed=&FN_FrontEndFloatModifier(
     Gameplay=&RandomRangeFloat(Minimum=200, Maximum=245)
     Front End=&RandomRangeFloat(Minimum=500, Maximum=650)
    )
   )
  ]
  particleUpdateScript [
   &ParticleState()
   &ScaleSpriteSize(Uniform Curve Sprite Scale=DataInterface() {propertyValues {Curve {keys [
        {arriveTangent=0.0398, interpMode=RCIM_Cubic, leaveTangent=0.0398, tangentMode=RCTM_User, time=0, value=0.6}
        {arriveTangent=0.3882, interpMode=RCIM_Cubic, leaveTangent=0.3882, tangentMode=RCTM_User, time=0.2, value=0.65}
        {arriveTangent=0.5569, interpMode=RCIM_Cubic, leaveTangent=0.5569, tangentMode=RCTM_User, time=1, value=1}
       ]}}})
   &ScaleColor(
    Scale Alpha=&FloatFromCurve(FloatCurve=DataInterface() {propertyValues {Curve {keys [
         {arriveTangent=0.7143, time=0.2, value=1}
         {arriveTangent=0.0076, interpMode=RCIM_Cubic, leaveTangent=0.0076, tangentMode=RCTM_User, time=0.7, value=1}
         {arriveTangent=-7.4459, interpMode=RCIM_Cubic, leaveTangent=-7.4459, tangentMode=RCTM_User, time=1, value=0}
        ]}}})
    Scale RGB=&Multiply_Vector(
     A=&MakeVectorFromLinearColorRGB(Color=&ColorFromCurve(ColorCurve=DataInterface() {propertyValues {
         AlphaCurve {keys [<at 0 1 RCIM_Cubic>]}
         BlueCurve {keys [<at 0 1 RCIM_Cubic>,<at 1 0.5882>]}
         GreenCurve {keys [<at 0 1 RCIM_Cubic>,<at 1 0.5882>]}
         RedCurve {keys [<at 0 1 RCIM_Cubic>,<at 1 0.5882>]}
        }}))
     B=&MakeVectorFromLinearColorRGB(Color=$System.TOD Color)
    )
    ScaleA=false
    ScaleRGB=true
   )
   &GravityForce(Gravity=&VectorFromCurve(
     VectorCurve=DataInterface() {propertyValues {
       XCurve {keys [<at 0 0>,<at 1 0>]}
       YCurve {keys [<at 0 0>,<at 1 0>]}
       ZCurve {keys [
         {arriveTangent=-0.1394, interpMode=RCIM_Cubic, leaveTangent=-0.1394, tangentMode=RCTM_User, time=0, value=0}
         {arriveTangent=9.9758, interpMode=RCIM_Cubic, leaveTangent=9.9758, tangentMode=RCTM_User, time=0.3, value=1}
        ]}
      }}
     Scale Curve=&RandomRangeVector(Minimum=<0 0 450>, Maximum=<0 0 575>)
    ))
   &Drag(
    Drag=&FloatFromCurve(
     FloatCurve=DataInterface() {propertyValues {Curve {keys [<at 0 0>,<at 0.05 1>]}}}
     Scale Curve=&RandomRangeFloat(Minimum=12, Maximum=16)
    )
    Use Linear Drag=true
    Use Rotational Drag=true
   )
   &InheritSourceMovement(
    Applied Position Scale=&VectorFromFloat(Value=&FloatFromCurve(
      FloatCurve=DataInterface() {propertyValues {Curve {keys [<at 0 0 RCIM_Cubic>, {arriveTangent=4.4031, interpMode=RCIM_Cubic, leaveTangent=4.4031, tangentMode=RCTM_User, time=0.1, value=1}]}}}
      Scale Curve=0.5
     ))
    ApplyForce=false
    ApplyPosition=true
    ApplyPositionFalloff=false
   )
   &SolveForcesAndVelocity(Clamp Velocity=false, Limit Acceleration=false)
   &DynamicMaterialParameters(
    Index 0 Param 1=&FloatFromCurve(FloatCurve=DataInterface() {propertyValues {Curve {keys [
         {arriveTangent=-0.026, interpMode=RCIM_Cubic, leaveTangent=-0.026, tangentMode=RCTM_User, time=0, value=0.2}
         {arriveTangent=2.4085, interpMode=RCIM_Cubic, leaveTangent=2.4085, tangentMode=RCTM_User, time=1, value=1}
        ]}}})
    Index 0 Param 2=&SelectFloatFromArray(
     Array Sampling Mode=::NewEnumerator0
     Direct Array Index=?
     Float Selection Array=DataInterface() {propertyValues {FloatData [0,1]}}
     Random Seed=0
    )
    Module.Index0Float0Write=true
    Module.Index0Float1Write=true
    Module.Index0Float2Write=true
    Module.Index0Float3Write=true
   ) {module=DynamicMaterialParameters001}
  ]
  rendererClasses [&NiagaraSpriteRendererProperties]
  renderers [{rendererClass=&NiagaraSpriteRendererProperties}]
 }
 FrontWheel {
  emitterSpawnScript [&NM_ToonShaderParameters(BaseColor:LinearColor=<rgba 0.6157 0.5412 0.3922 1>, Lighting_Hardness=0.2)]
  emitterUpdateScript [
   &EmitterState(Life Cycle Mode=::NewEnumerator1, Loop Behavior=::NewEnumerator1, Loop Delay=2.2, Loop Duration Mode=::NewEnumerator1, UseLoopDelay=false)
   &SpawnRate(
    SpawnRate=&FN_FrontEndFloatModifier(Gameplay=&Lerp_Float(A=0, B=64, Alpha=$System.Speed Threshold), Front End=64)
    Use Spawn Probability=false
   )
   &SpawnPerUnit(Module.Use Max Movement Threshold=true, Module.Use Movement Tolerance=false, Spawn Spacing=5, Use Spawn Probability=false) {enabled=false}
  ]
  particleSpawnScript [
   &InitializeParticle(
    Color=&Multiply_LinearColorByFloat(
     Linear Color=&FNTimeOfDayColor(Night Color=<rgba 0.7961 0.8667 1 1>, Morning Color=<rgba 1 1 1 1>, Day Color=<rgba 1 1 1 1>, Evening Color=<rgba 1 1 1 1>)
     Float=&MakeFloatFromLinearColor(
      LinearColor=&MakeCustomLinearColorFromBool(
       Bool=&SetBoolByFloatComparison(A=$User.FrontEnd, B=0)
       True Color=<rgba 1 1 1 1>
       False Color=<rgba 0.6 0.6 0.6 1>
      )
      Channel:ENiagaraLinearColor_Channels=::NewEnumerator0
     )
    )
    Lifetime Max=0.45
    Lifetime Min=0.3
    Lifetime Mode=::NewEnumerator1
    Mesh Scale=&Multiply_VectorByFloat(Vector=<1 1 1>, Float=&RandomRangeFloat(Minimum=0.5, Maximum=0.7))
    Mesh Scale Mode=::NewEnumerator4
    UsePositionOffset=false
    Write Material Random=true
   )
   &FN_SetSpawnGroup(Number of Spawn Groups=2)
   &SkeletalMeshLocation(
    Bone / Socket Index=&SpawnGroup()
    Bone Sampling Mode=::NewEnumerator4
    MaskSpawnGroup=false
    Sampled Position Offset=/Game/Effects/Niagara/DynamicInputs/Water/FN_PickVectorFromSpawnGroupCutoff(
     Under Cutoff Vector:Vector3f=<7 -8 -5>
     Over Cutoff Vector:Vector3f=<-7 -8 -5>
     SpawnGroup:NiagaraInt32=$Particles.SpawnGroup
     Cutoff:NiagaraInt32=1
    )
    Sampled Position Offset Coordinate Space:ENiagaraSkelMeshTransforms=::NewEnumerator2
    Skeletal Mesh=DataInterface() {propertyValues {
      DefaultMesh=null
      ExcludeBoneName=root
      OptionalComponentClass=null
      PreviewMesh=/BRCosmetics/Animation/Props/Gadgets_ChiveFlake/Meshes/ChiveFlake_Gadget
      SkinningMode=SkinOnTheFly
      SoftSourceActor=null
      SourceMode=Default
      UvSetIndex=0
      WholeMeshLOD=-1
      bExcludeBone=false
      bReadDeformedGeometry=true
      bRequireCurrentFrameData=true
      ComponentTags []
      FilteredBones [JNT_FrontWheelLeft, JNT_FrontWheelRight]
      FilteredSockets []
      MeshUserParameter {parameter {name=None, typeDefHandle {registeredTypeIndex=1}, varData []}}
      SamplingRegions []
     }}
   )
   &InitialMeshOrientation(Rotation=&RandomRangeVector(Minimum=<0 0 0>, Maximum=<1 1 1>), Use Rotation Vector=true)
   &ShapeLocation(MaskSpawnGroup=false, Override Seed=false, Sphere Radius=2, UseOwnerScale=true)
   &AddVelocity(
    Cone Angle=&FN_FrontEndFloatModifier(Gameplay=60, Front End=30)
    Cone Axis=<0 -1 0>
    Module.UseConeFalloff=false
    Module.UseVelDistribution=false
    Override Seed=false
    Velocity Mode=::NewEnumerator2
    Velocity Speed=&FN_FrontEndFloatModifier(
     Gameplay=&RandomRangeFloat(Minimum=120, Maximum=160)
     Front End=&RandomRangeFloat(Minimum=320, Maximum=450)
    )
   )
  ]
  particleUpdateScript [
   &ParticleState()
   &ScaleMeshSize(Scale Factor=&VectorFromFloat(Value=&FloatFromCurve(FloatCurve=DataInterface() {propertyValues {Curve {keys [
          {arriveTangent=0.0398, interpMode=RCIM_Cubic, leaveTangent=0.0398, tangentMode=RCTM_User, time=0, value=0.4}
          {arriveTangent=0.4614, interpMode=RCIM_Cubic, leaveTangent=0.4614, tangentMode=RCTM_User, time=0.2, value=0.475}
          {arriveTangent=2.388, interpMode=RCIM_Cubic, leaveTangent=2.388, tangentMode=RCTM_User, time=1, value=1.6}
         ]}}})))
   &ScaleColor(
    Scale Alpha=&FloatFromCurve(FloatCurve=DataInterface() {propertyValues {Curve {keys [
         {arriveTangent=0.7143, time=0.2, value=1}
         {arriveTangent=0.0076, interpMode=RCIM_Cubic, leaveTangent=0.0076, tangentMode=RCTM_User, time=0.7, value=1}
         {arriveTangent=-7.4459, interpMode=RCIM_Cubic, leaveTangent=-7.4459, tangentMode=RCTM_User, time=1, value=0}
        ]}}})
    Scale RGB=&Multiply_Vector(
     A=&MakeVectorFromLinearColorRGB(Color=&ColorFromCurve(ColorCurve=DataInterface() {propertyValues {
         AlphaCurve {keys [<at 0 1 RCIM_Cubic>]}
         BlueCurve {keys [<at 0 1 RCIM_Cubic>,<at 1 0.5882>]}
         GreenCurve {keys [<at 0 1 RCIM_Cubic>,<at 1 0.5882>]}
         RedCurve {keys [<at 0 1 RCIM_Cubic>,<at 1 0.5882>]}
        }}))
     B=&MakeVectorFromLinearColorRGB(Color=$System.TOD Color)
    )
    ScaleA=false
    ScaleRGB=true
   )
   &GravityForce(Gravity=&VectorFromCurve(
     VectorCurve=DataInterface() {propertyValues {
       XCurve {keys [<at 0 0>,<at 1 0>]}
       YCurve {keys [<at 0 0>,<at 1 0>]}
       ZCurve {keys [
         {arriveTangent=-0.1394, interpMode=RCIM_Cubic, leaveTangent=-0.1394, tangentMode=RCTM_User, time=0, value=0}
         {arriveTangent=9.9758, interpMode=RCIM_Cubic, leaveTangent=9.9758, tangentMode=RCTM_User, time=0.3, value=1}
        ]}
      }}
     Scale Curve=&RandomRangeVector(Minimum=<0 0 100>, Maximum=<0 0 120>)
    ))
   &Drag(
    Drag=&FloatFromCurve(
     FloatCurve=DataInterface() {propertyValues {Curve {keys [<at 0 0>,<at 0.05 1>]}}}
     Scale Curve=&RandomRangeFloat(Minimum=20, Maximum=24)
    )
    Use Linear Drag=true
    Use Rotational Drag=true
   )
   &InheritSourceMovement(
    Applied Position Scale=&VectorFromFloat(Value=&FloatFromCurve(
      FloatCurve=DataInterface() {propertyValues {Curve {keys [<at 0 0 RCIM_Cubic>, {arriveTangent=4.4031, interpMode=RCIM_Cubic, leaveTangent=4.4031, tangentMode=RCTM_User, time=0.1, value=1}]}}}
      Scale Curve=0.7
     ))
    ApplyForce=false
    ApplyPosition=true
    ApplyPositionFalloff=false
   )
   &SolveForcesAndVelocity(Clamp Velocity=false, Limit Acceleration=false)
   &DynamicMaterialParameters(
    Index 0 Param 1=&FloatFromCurve(FloatCurve=DataInterface() {propertyValues {Curve {keys [
         {arriveTangent=-0.0276, time=0, value=0}
         {arriveTangent=0.0397, interpMode=RCIM_Cubic, leaveTangent=0.0397, tangentMode=RCTM_User, time=0.65, value=0}
         {arriveTangent=3.5539, interpMode=RCIM_Cubic, leaveTangent=3.5539, tangentMode=RCTM_User, time=1, value=0.4}
        ]}}})
    Module.Index0Float0Write=true
    Module.Index0Float1Write=true
    Module.Index0Float2Write=true
    Module.Index0Float3Write=true
   ) {enabled=false}
   &DynamicMaterialParameters(
    Index 0 Param 1=&FloatFromCurve(FloatCurve=DataInterface() {propertyValues {Curve {keys [
         {arriveTangent=-0.0276, time=0, value=0}
         {arriveTangent=0.0397, interpMode=RCIM_Cubic, leaveTangent=0.0397, tangentMode=RCTM_User, time=0.45, value=0}
         {arriveTangent=2.6349, interpMode=RCIM_Cubic, leaveTangent=2.6349, tangentMode=RCTM_User, time=1, value=0.6}
        ]}}})
    Module.Index0Float0Write=true
    Module.Index0Float1Write=true
    Module.Index0Float2Write=true
    Module.Index0Float3Write=true
   ) {module=DynamicMaterialParameters001}
   &UpdateMeshOrientation(
    Rotation Rate=&FloatFromCurve(
     FloatCurve=DataInterface() {propertyValues {Curve {keys [{interpMode=RCIM_Cubic, tangentMode=RCTM_User, time=0.4, value=0}, {interpMode=RCIM_Cubic, tangentMode=RCTM_User, time=1, value=1}]}}}
     Scale Curve=&RandomRangeFloat(Minimum=-2, Maximum=2)
    )
    Rotation Vector=&RandomVector(Vector Scale=1)
   )
  ]
  rendererClasses [&NiagaraMeshRendererProperties]
  renderers [
   {rendererClass=&NiagaraMeshRendererProperties, rendererIndex=0}
   {rendererClass=&NiagaraMeshRendererProperties, rendererIndex=1}
  ]
 }
 FrontWheelSPRITE {
  emitterUpdateScript [
   &EmitterState(Life Cycle Mode=::NewEnumerator1, Loop Behavior=::NewEnumerator1, Loop Delay=2.2, Loop Duration Mode=::NewEnumerator1, UseLoopDelay=false)
   &SpawnRate(
    SpawnRate=&FN_FrontEndFloatModifier(Gameplay=&Lerp_Float(A=0, B=32, Alpha=$System.Speed Threshold), Front End=64)
    Use Spawn Probability=false
   )
   &SpawnPerUnit(Module.Use Max Movement Threshold=true, Module.Use Movement Tolerance=false, Spawn Spacing=5, Use Spawn Probability=false) {enabled=false}
  ]
  particleSpawnScript [
   &InitializeParticle(
    Color=&MultiplyLinearColor(
     Linear Color=<rgba 0.9569 0.8275 0.5647 1>
     Scale Factor=&MakeVector4_FromLinearColor(LinearColor=&FNTimeOfDayColor(Night Color=<rgba 0.5765 0.6157 1 1>, Morning Color=<rgba 1 1 1 1>, Day Color=<rgba 1 1 1 1>, Evening Color=<rgba 1 1 1 1>))
    )
    Lifetime Max=0.45
    Lifetime Min=0.3
    Lifetime Mode=::NewEnumerator1
    Sprite Rotation Mode=::NewEnumerator2
    Sprite Size Mode=::NewEnumerator1
    Uniform Sprite Size Max=35
    Uniform Sprite Size Min=20
    UsePositionOffset=false
    Write Material Random=true
   )
   &FN_SetSpawnGroup(Number of Spawn Groups=2)
   &SkeletalMeshLocation(
    Bone / Socket Index=&SpawnGroup()
    Bone Sampling Mode=::NewEnumerator4
    MaskSpawnGroup=false
    Sampled Position Offset=&FN_PickVectorFromSpawnGroupCutoff(
     Under Cutoff Vector:Vector3f=<7 -8 -5>
     Over Cutoff Vector:Vector3f=<-7 -8 -5>
     SpawnGroup:NiagaraInt32=$Particles.SpawnGroup
     Cutoff:NiagaraInt32=1
    )
    Sampled Position Offset Coordinate Space:ENiagaraSkelMeshTransforms=::NewEnumerator2
    Skeletal Mesh=DataInterface() {propertyValues {
      DefaultMesh=null
      ExcludeBoneName=root
      OptionalComponentClass=null
      PreviewMesh=&ChiveFlake_Gadget
      SkinningMode=SkinOnTheFly
      SoftSourceActor=null
      SourceMode=Default
      UvSetIndex=0
      WholeMeshLOD=-1
      bExcludeBone=false
      bReadDeformedGeometry=true
      bRequireCurrentFrameData=true
      ComponentTags []
      FilteredBones [JNT_FrontWheelLeft, JNT_FrontWheelRight]
      FilteredSockets []
      MeshUserParameter {parameter {name=None, typeDefHandle {registeredTypeIndex=1}, varData []}}
      SamplingRegions []
     }}
   )
   &ShapeLocation(MaskSpawnGroup=false, Override Seed=false, Sphere Radius=2, UseOwnerScale=true)
   &AddVelocity(
    Cone Angle=&FN_FrontEndFloatModifier(Gameplay=60, Front End=30)
    Cone Axis=<0 -1 0>
    Module.UseConeFalloff=false
    Module.UseVelDistribution=false
    Override Seed=false
    Velocity Mode=::NewEnumerator2
    Velocity Speed=&FN_FrontEndFloatModifier(
     Gameplay=&RandomRangeFloat(Minimum=120, Maximum=160)
     Front End=&RandomRangeFloat(Minimum=320, Maximum=450)
    )
   )
  ]
  particleUpdateScript [
   &ParticleState()
   &ScaleSpriteSize(Uniform Curve Sprite Scale=DataInterface() {propertyValues {Curve {keys [
        {arriveTangent=0.0398, interpMode=RCIM_Cubic, leaveTangent=0.0398, tangentMode=RCTM_User, time=0, value=0.6}
        {arriveTangent=0.3882, interpMode=RCIM_Cubic, leaveTangent=0.3882, tangentMode=RCTM_User, time=0.2, value=0.65}
        {arriveTangent=0.5569, interpMode=RCIM_Cubic, leaveTangent=0.5569, tangentMode=RCTM_User, time=1, value=1}
       ]}}})
   &ScaleColor(
    Scale Alpha=&FloatFromCurve(FloatCurve=DataInterface() {propertyValues {Curve {keys [
         {arriveTangent=0.7143, time=0.2, value=1}
         {arriveTangent=0.0076, interpMode=RCIM_Cubic, leaveTangent=0.0076, tangentMode=RCTM_User, time=0.7, value=1}
         {arriveTangent=-7.4459, interpMode=RCIM_Cubic, leaveTangent=-7.4459, tangentMode=RCTM_User, time=1, value=0}
        ]}}})
    Scale RGB=&Multiply_Vector(
     A=&MakeVectorFromLinearColorRGB(Color=&ColorFromCurve(ColorCurve=DataInterface() {propertyValues {
         AlphaCurve {keys [<at 0 1 RCIM_Cubic>]}
         BlueCurve {keys [<at 0 1 RCIM_Cubic>,<at 1 0.5882>]}
         GreenCurve {keys [<at 0 1 RCIM_Cubic>,<at 1 0.5882>]}
         RedCurve {keys [<at 0 1 RCIM_Cubic>,<at 1 0.5882>]}
        }}))
     B=&MakeVectorFromLinearColorRGB(Color=$System.TOD Color)
    )
    ScaleA=false
    ScaleRGB=true
   )
   &GravityForce(Gravity=&VectorFromCurve(
     VectorCurve=DataInterface() {propertyValues {
       XCurve {keys [<at 0 0>,<at 1 0>]}
       YCurve {keys [<at 0 0>,<at 1 0>]}
       ZCurve {keys [
         {arriveTangent=-0.1394, interpMode=RCIM_Cubic, leaveTangent=-0.1394, tangentMode=RCTM_User, time=0, value=0}
         {arriveTangent=9.9758, interpMode=RCIM_Cubic, leaveTangent=9.9758, tangentMode=RCTM_User, time=0.3, value=1}
        ]}
      }}
     Scale Curve=&RandomRangeVector(Minimum=<0 0 100>, Maximum=<0 0 120>)
    ))
   &Drag(
    Drag=&FloatFromCurve(
     FloatCurve=DataInterface() {propertyValues {Curve {keys [<at 0 0>,<at 0.05 1>]}}}
     Scale Curve=&RandomRangeFloat(Minimum=20, Maximum=24)
    )
    Use Linear Drag=true
    Use Rotational Drag=true
   )
   &InheritSourceMovement(
    Applied Position Scale=&VectorFromFloat(Value=&FloatFromCurve(
      FloatCurve=DataInterface() {propertyValues {Curve {keys [<at 0 0 RCIM_Cubic>, {arriveTangent=4.4031, interpMode=RCIM_Cubic, leaveTangent=4.4031, tangentMode=RCTM_User, time=0.1, value=1}]}}}
      Scale Curve=0.7
     ))
    ApplyForce=false
    ApplyPosition=true
    ApplyPositionFalloff=false
   )
   &SolveForcesAndVelocity(Clamp Velocity=false, Limit Acceleration=false)
   &DynamicMaterialParameters(
    Index 0 Param 1=&FloatFromCurve(FloatCurve=DataInterface() {propertyValues {Curve {keys [
         {arriveTangent=-0.026, interpMode=RCIM_Cubic, leaveTangent=-0.026, tangentMode=RCTM_User, time=0, value=0.2}
         {arriveTangent=2.4085, interpMode=RCIM_Cubic, leaveTangent=2.4085, tangentMode=RCTM_User, time=1, value=1}
        ]}}})
    Index 0 Param 2=&SelectFloatFromArray(
     Array Sampling Mode=::NewEnumerator0
     Direct Array Index=?
     Float Selection Array=DataInterface() {propertyValues {FloatData [0,1]}}
     Random Seed=0
    )
    Module.Index0Float0Write=true
    Module.Index0Float1Write=true
    Module.Index0Float2Write=true
    Module.Index0Float3Write=true
   ) {module=DynamicMaterialParameters001}
  ]
  rendererClasses [&NiagaraSpriteRendererProperties]
  renderers [{rendererClass=&NiagaraSpriteRendererProperties}]
 }
 Rocks {
  emitterSpawnScript [&NM_ToonShaderParameters(BaseColor:LinearColor=<rgba 0.1615 0.141 0.1015 1>, Lighting_Hardness=0.2, RimLight_Color:LinearColor=<rgba 0.9896 0.9896 0.9896 1>)]
  emitterUpdateScript [
   &EmitterState(Life Cycle Mode=::NewEnumerator1, Loop Behavior=::NewEnumerator1, Loop Duration Mode=::NewEnumerator1, UseLoopDelay=false)
   &SpawnRate(SpawnRate=&FN_FrontEndFloatModifier(Gameplay=0, Front End=12), Use Spawn Probability=false)
   &SpawnPerUnit(Module.Use Max Movement Threshold=true, Module.Use Movement Tolerance=false, Use Spawn Probability=false)
  ]
  particleSpawnScript [
   &InitializeParticle(
    Color=&FN_Select_Color_from_Time_Of_Day(Morning=<rgba 1 1 1 1>, Day=<rgba 1 1 1 1>, Evening=<rgba 1 1 1 1>, Night=<rgba 0.2294 0.2568 0.4635 1>)
    Lifetime Max=0.9
    Lifetime Min=0.7
    Lifetime Mode=::NewEnumerator1
    Mass Mode=::NewEnumerator3
    Mesh Scale Mode=::NewEnumerator1
    Mesh Uniform Scale Max=0.4
    Mesh Uniform Scale Min=0.3
    Sprite Rotation Mode=::NewEnumerator2
    Sprite Size Mode=::NewEnumerator1
    Uniform Sprite Size Max=8
    UsePositionOffset=false
    Write Material Random=true
   )
   &FN_SetSpawnGroup(Number of Spawn Groups=2)
   &SkeletalMeshLocation(
    Bone / Socket Index=&SpawnGroup()
    Bone Sampling Mode=::NewEnumerator4
    MaskSpawnGroup=false
    Skeletal Mesh=DataInterface() {propertyValues {
      DefaultMesh=null
      ExcludeBoneName=root
      OptionalComponentClass=null
      PreviewMesh=&ChiveFlake_Gadget
      SkinningMode=SkinOnTheFly
      SoftSourceActor=null
      SourceMode=Default
      UvSetIndex=0
      WholeMeshLOD=-1
      bExcludeBone=false
      bReadDeformedGeometry=true
      bRequireCurrentFrameData=true
      ComponentTags []
      FilteredBones [JNT_BackWheelRight, JNT_BackWheelLeft]
      FilteredSockets []
      MeshUserParameter {parameter {name=None, typeDefHandle {registeredTypeIndex=1}, varData []}}
      SamplingRegions []
     }}
   )
   &ShapeLocation(MaskSpawnGroup=false, Override Seed=false, Sphere Radius=3, UseOwnerScale=true)
   &InitialMeshOrientation(Use Rotation Vector=false)
   &AddVelocity(
    Cone Angle=&FN_FrontEndFloatModifier(Gameplay=90, Front End=30)
    Cone Axis=&FN_FrontEndVector3Modifier(Gameplay=<0 -0.2 1>, Front End=<0 -0.8 1>)
    Distribution Along Cone Axis=0.25
    Module.UseConeFalloff=true
    Module.UseVelDistribution=true
    Override Seed=false
    Speed Falloff From Cone Axis=0.5
    Velocity Mode=::NewEnumerator2
    Velocity Speed=&FN_FrontEndFloatModifier(
     Gameplay=&RandomRangeFloat(Minimum=110, Maximum=300)
     Front End=&RandomRangeFloat(Minimum=200, Maximum=300)
    )
   )
   &AddRotationalVelocity(Rotation Rate=&VectorFromFloat(Value=&RandomRangeFloat(Minimum=15, Maximum=30)))
  ]
  particleUpdateScript [
   &ParticleState()
   &ScaleMeshSize(Scale Factor=&VectorFromFloat(Value=&FloatFromCurve(FloatCurve=DataInterface() {propertyValues {Curve {keys [
          {tangentMode=RCTM_User, time=0, value=1}
          {arriveTangent=-0.0405, interpMode=RCIM_Cubic, leaveTangent=-0.0405, tangentMode=RCTM_User, time=0.7009, value=1}
          {arriveTangent=-8.0089, interpMode=RCIM_Cubic, leaveTangent=-8.0089, tangentMode=RCTM_User, time=1, value=0}
         ]}}})))
   &GravityForce()
   &Collision(
    Age Colliding Particles=false
    Collision Enabled=&Boolean_AndOperation(
     Bool1=&IsPlatformSetActive(New Platform Set=DataInterface() {propertyValues {Platforms {qualityLevelMask=-4}}})
     Bool2=&SetBoolByIntComparison(A=&RandomRangeInt(Minimum=0, Maximum=0), B=0, Comparison Type=::NewEnumerator4)
    )
    Control Roll On Collision=true
    KillOnCollision=false
    Module.Control Rest State Via Surface Normal=false
    Module.EnableMaxCollisionCount=false
    Module.Randomize Collision Normal Vector=false
    Restitution=0.25
    ScaleRateOfRotation=false
   )
   &Drag(Drag=&RandomRangeFloat(Minimum=0.8, Maximum=1.2), Use Linear Drag=true, Use Rotational Drag=false)
   &ScaleColor(
    Scale Alpha=&FloatFromCurve(FloatCurve=DataInterface() {propertyValues {Curve {keys [
         <at -0.0046 1.0113 RCIM_Cubic>
         {arriveTangent=-3.6957, interpMode=RCIM_Cubic, leaveTangent=-3.6957, tangentMode=RCTM_User, time=1.0011, value=0.002}
        ]}}})
    Scale RGB=&VectorFromFloat(Value=1)
    ScaleA=true
    ScaleRGB=true
   )
   &SolveForcesAndVelocity(Clamp Velocity=false, Limit Acceleration=false)
  ]
  rendererClasses [&NiagaraMeshRendererProperties]
  renderers [{rendererClass=&NiagaraMeshRendererProperties, rendererIndex=0}]
 }
}
summary {
 systemName=NS_ChiveFlake_CarFX
 emitters [
  {emitterName=Rocks, rendererClasses [&NiagaraMeshRendererProperties]}
  {emitterName=BackWheel, rendererClasses [&NiagaraMeshRendererProperties]}
  {emitterName=FrontWheel, rendererClasses [&NiagaraMeshRendererProperties]}
  {emitterName=BackWheelSPRITE, rendererClasses [&NiagaraSpriteRendererProperties]}
  {emitterName=FrontWheelSPRITE, rendererClasses [&NiagaraSpriteRendererProperties]}
 ]
 userVariables [
  User.bIsFrontEnd {type=NiagaraBool}
  User.bIsFrontEndPreview {type=NiagaraBool}
  User.bIsInWater {type=NiagaraBool}
  User.FrontEnd {type=NiagaraFloat, defaultValue=0, description=""}
 ]
}
